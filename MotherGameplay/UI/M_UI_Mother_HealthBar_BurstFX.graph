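# Translucent UI material for a burst effect on the Mother health bar, built from panning blurred noise.
# The first noise layer offsets the second's UVs, then the inverted second layer masks the first.
# AnimateSubtract moves a shared threshold between start and end values to erode both layers.
# The resulting mask drives a three-color blend and, faded by an exponential radial gradient and scaled by Final_Alpha, an alpha value.
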Material /MotherGameplay/UI/M_UI_Mother_HealthBar_BurstFX
graph {
 properties {blendMode=BLEND_Translucent, materialDomain=MD_UI, pixelDepthOffsetMode=PDOM_Legacy, usedShadingModels=Default Lit}
 nodes [
  MaterialFunctionCall() {materialFunction=&VectorToRadialValue}
  n1 = ScalarParameter() {defaultValue=0.5, group=Noise2, parameterName=Scale_Texture_2_X}
  n2 = ScalarParameter() {defaultValue=2, group=Noise2, parameterName=Scale_Texture_2_Y}
  n3 = TextureCoordinate()
  n4 = Add(
   A=MaterialFunctionCall(UVs=@n3, Texture Scale=AppendVector(A=@n1, B=@n2)) {materialFunction=&ScaleUVsByCenter}.UVs
   B=AppendVector(A=ScalarParameter() {group=Noise2, parameterName=Offset_X_Texture_2}, B=ScalarParameter() {group=Noise2, parameterName=Offset_Y_Texture_2})
  )
  n5 = AppendVector(
   A=ScalarParameter() {group=Noise2, parameterName=PanSpeed_X_Texture_2}
   B=ScalarParameter() {defaultValue=0.125, group=Noise2, parameterName=PanSpeedt_Y_Texture_2}
  )
  n6 = MaterialFunctionCall(
   UVs=@n3
   Texture Scale=AppendVector(
    A=ScalarParameter() {defaultValue=0.75, group=Noise1, parameterName=Scale_Texture_1_X}
    B=ScalarParameter() {defaultValue=0.75, group=Noise1, parameterName=Scale_Texture_1_Y}
   )
  ) {materialFunction=&ScaleUVsByCenter}
  n7 = AppendVector(A=ScalarParameter() {group=Noise1, parameterName=Offset_X_Texture_1}, B=ScalarParameter() {group=Noise1, parameterName=Offset_Y_Texture_1})
  n8 = ScalarParameter() {group=Noise1, parameterName=PanSpeed_X_Texture_1}
  n9 = ScalarParameter() {defaultValue=0.125, group=Noise1, parameterName=PanSpeedt_Y_Texture_1}
  n10 = TextureSample(UVs=Panner(Coordinate=Add(A=@n6.UVs, B=@n7), Speed=AppendVector(A=@n8, B=@n9))) {samplerType=SAMPLERTYPE_Grayscale, texture=&LowResBlurredNoise}
  n11 = TextureSample(UVs=Add(A=Panner(Coordinate=@n4, Speed=@n5), B=@n10.R)) {samplerType=SAMPLERTYPE_Grayscale, texture=&LowResBlurredNoise}
  n12 = LinearInterpolate(
   A=ScalarParameter() {group=Global, parameterName=Subtract_Start}
   B=ScalarParameter() {defaultValue=1, group=Global, parameterName=Subtract_End}
   Alpha=ScalarParameter() {group=Global, parameterName=AnimateSubtract}
  )
  n13 = Saturate(None=Multiply(A=Subtract(A=@n11.R, B=@n12), B=ScalarParameter() {defaultValue=8, group=Noise2, parameterName=Lift_Texture_2}))
  n14 = Multiply(A=Subtract(A=@n10.R, B=@n12), B=ScalarParameter() {defaultValue=3, group=Noise1, parameterName=Lift_Texture_1})
  n15 = Saturate(None=DotProduct(A=OneMinus(None=@n13), B=Saturate(None=@n14)))
  MaterialFunctionCall(
   A=VectorParameter() {group=Global, parameterName=Color_1, defaultValue=<rgba 1 0 0 1>}.RGB
   B=VectorParameter() {group=Global, parameterName=Color_2, defaultValue=<rgba 0 1 0 1>}.RGB
   C=VectorParameter() {group=Global, parameterName=Color_3, defaultValue=<rgba 0 0 1 1>}.RGB
   Alpha=@n15
  ) {materialFunction=&3ColorBlend}
  n16 = MaterialFunctionCall(
   Radius=ScalarParameter() {defaultValue=0.5, group=Global, parameterName=Radius}
   Density=ScalarParameter() {defaultValue=2.33, group=Global, parameterName=Density}
  ) {materialFunction=&RadialGradientExponential}
  Multiply(
   A=Saturate(None=DotProduct(A=@n15, B=@n16.RadialGradientExponential))
   B=ScalarParameter() {defaultValue=1, group=Global, parameterName=Final_Alpha}
  )
  ScalarParameter() {defaultValue=0.333, group=Noise2, parameterName=Subtract_Texture_1}
  ScalarParameter() {defaultValue=0.333, group=Noise1, parameterName=Subtract_Texture_2}
  Sine(None=@n13)
 ]
}
