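# Translucent UI master material for the Mother health bar indicator, drawing a wobbling parasite gauge bracket over an animated glow.
# Panning noise, warped by static noise, forms the glow within side and vertical masks, with AnimateGlow blending toward a burst color.
# Opposed panning sines and a time sine offset the gauge texture vertically.
# Its alpha composites the bracket, tinted by a center-to-side gradient and noise spots, over the glow.
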Material /MotherGameplay/UI/M_UI_Mother_HealthBar_Indicator_Master
graph {
 properties {blendMode=BLEND_Translucent, materialDomain=MD_UI, pixelDepthOffsetMode=PDOM_Legacy, usedShadingModels=Default Lit}
 nodes [
  n1 = VectorParameter() {group=Glow, parameterName=Glow_Color_Falloff, defaultValue=<rgba 0.6167 0 1 1>}
  n2 = VectorParameter() {group=Glow, parameterName=Glow_Color_Falloff_Burst, defaultValue=<rgba 0.963 0.4316 1 1>}
  n3 = ScalarParameter() {parameterName=AnimateGlow}
  n4 = TextureCoordinate()
  n5 = MaterialFunctionCall(
   UVs=@n4
   Texture Scale=AppendVector(
    A=ScalarParameter() {defaultValue=0.5, group=GlowNoise, parameterName=Scale_Texture_2_X}
    B=ScalarParameter() {defaultValue=2, group=GlowNoise, parameterName=Scale_Texture_2_Y}
   )
  ) {materialFunction=&ScaleUVsByCenter}
  n6 = AppendVector(
   A=ScalarParameter() {group=GlowNoise, parameterName=Offset_X_Texture_2}
   B=ScalarParameter() {group=GlowNoise, parameterName=Offset_Y_Texture_2}
  )
  n7 = ScalarParameter() {group=GlowNoise, parameterName=PanSpeed_X_Texture_2}
  n8 = ScalarParameter() {defaultValue=0.125, group=GlowNoise, parameterName=PanSpeedt_Y_Texture_2}
  n9 = AppendVector(
   A=ScalarParameter() {defaultValue=0.75, group=Pattern, parameterName=Scale_Texture_1_X}
   B=ScalarParameter() {defaultValue=0.75, group=Pattern, parameterName=Scale_Texture_1_Y}
  )
  n10 = ScalarParameter() {group=Pattern, parameterName=Offset_X_Texture_1}
  n11 = ScalarParameter() {group=Pattern, parameterName=Offset_Y_Texture_1}
  n12 = TextureSample(UVs=Add(
    A=MaterialFunctionCall(UVs=@n4, Texture Scale=@n9) {materialFunction=&ScaleUVsByCenter}.UVs
    B=AppendVector(A=@n10, B=@n11)
   )) {samplerType=SAMPLERTYPE_Grayscale, texture=&LowResBlurredNoise}
  n13 = Add(A=Panner(Coordinate=Add(A=@n5.UVs, B=@n6), Speed=AppendVector(A=@n7, B=@n8)), B=@n12.R)
  n14 = Multiply(
   A=Subtract(
    A=TextureSample(UVs=@n13) {samplerType=SAMPLERTYPE_Grayscale, texture=&LowResBlurredNoise}.R
    B=ScalarParameter() {defaultValue=0.333, group=GlowNoise, parameterName=Subtract_Texture_2}
   )
   B=ScalarParameter() {defaultValue=8, group=GlowNoise, parameterName=Lift_Texture_2}
  )
  n15 = Sine(None=OneMinus(None=Saturate(None=@n14)))
  n16 = TextureCoordinate()
  n17 = ComponentMask(None=Sine(None=@n16) {period=2}) {r=true}
  n18 = ScalarParameter() {defaultValue=0.025, group=Glow, parameterName=ClampSides}
  n19 = Saturate(None=Multiply(A=Subtract(A=@n17, B=@n18), B=ScalarParameter() {defaultValue=1, group=Glow, parameterName=SharpenSides}))
  n20 = ComponentMask(None=@n16) {g=true}
  n21 = Add(A=@n16, B=ScalarParameter() {defaultValue=0.333, group=Glow, parameterName=OffsetGlow})
  n22 = OneMinus(None=ComponentMask(None=@n21) {g=true})
  n23 = ScalarParameter() {defaultValue=80, group=Glow, parameterName=Glow_Bottom_Edge_Hardness}
  n24 = DotProduct(A=@n20, B=Saturate(None=Multiply(A=@n22, B=@n23)))
  n25 = Saturate(None=DotProduct(A=@n19, B=Saturate(None=@n24)))
  n26 = Multiply(A=@n25, B=ScalarParameter() {defaultValue=0.5, group=Glow, parameterName=Glow_Fade_Final_Alpha})
  n27 = Add(A=DotProduct(A=Saturate(None=@n15), B=@n25), B=@n26)
  n28 = Constant()
  n29 = ScalarParameter() {defaultValue=0.5, group=Distortion, parameterName=CenterDistortionSpeed}
  n30 = DotProduct(
   A=@n17
   B=Saturate(None=MaterialFunctionCall(Speed=@n29) {materialFunction=&DebugTimeSine}.Result)
  )
  n31 = TextureCoordinate()
  n32 = AppendVector(A=ScalarParameter() {defaultValue=0.33, group=Distortion, parameterName=PanSpeed_X_Sine_1}, B=Constant())
  n33 = Abs(None=Sine(None=Panner(Coordinate=@n31, Speed=@n32) {speedX=0.125}) {period=2})
  n34 = AppendVector(A=ScalarParameter() {defaultValue=0.5, group=Distortion, parameterName=PanSpeed_X_Sine_2}, B=Constant())
  n35 = Abs(None=Sine(None=Panner(Coordinate=@n31, Speed=@n34)) {period=1.5})
  n36 = Saturate(None=DotProduct(A=ComponentMask(None=@n33) {r=true}, B=ComponentMask(None=@n35) {r=true}))
  n37 = Saturate(None=LinearInterpolate(Alpha=ComponentMask(None=@n31) {r=true}) {constB=2})
  n38 = Saturate(None=LinearInterpolate(B=LinearInterpolate(Alpha=@n36) {constA=0.5}, Alpha=OneMinus(None=@n37)))
  n39 = TextureCoordinate()
  n40 = AppendVector(A=ScalarParameter() {defaultValue=-0.33, group=Distortion, parameterName=PanSpeed_X_Sine_3}, B=Constant())
  n41 = Abs(None=Sine(None=Panner(Coordinate=@n39, Speed=@n40) {speedX=0.125}) {period=2})
  n42 = AppendVector(A=ScalarParameter() {defaultValue=-0.4, group=Distortion, parameterName=PanSpeed_X_Sine_4}, B=Constant())
  n43 = Abs(None=Sine(None=Panner(Coordinate=@n39, Speed=@n42)) {period=1.5})
  n44 = Saturate(None=DotProduct(A=ComponentMask(None=@n41) {r=true}, B=ComponentMask(None=@n43) {r=true}))
  n45 = OneMinus(None=@n39)
  n46 = Saturate(None=LinearInterpolate(Alpha=ComponentMask(None=@n45) {r=true}) {constB=2})
  n47 = Saturate(None=LinearInterpolate(B=LinearInterpolate(Alpha=@n44) {constA=0.5}, Alpha=OneMinus(None=@n46)))
  n48 = LinearInterpolate(
   A=Multiply(A=Saturate(None=@n30), B=ScalarParameter() {defaultValue=0.5, group=Distortion, parameterName=CenterDistortion_Amount})
   Alpha=Saturate(None=Add(A=@n38, B=@n47))
  )
  n49 = Multiply(A=Saturate(None=DotProduct(A=@n19, B=@n48)), B=ScalarParameter() {defaultValue=0.05, group=Distortion, parameterName=DistortionAmount})
  n50 = TextureSample(UVs=Add(A=@n16, B=AppendVector(A=@n28, B=@n49))) {samplerSource=SSM_Clamp_WorldGroupSettings, texture=/MotherGameplay/UI/Textures/T_UI_ParasiteGauge}
  n51 = Saturate(None=Subtract(A=Saturate(None=@n50.A), B=ScalarParameter()))
  n52 = VectorParameter() {group=Bracket, parameterName=Bracket_Color_2, defaultValue=<rgba 0.0125 0 0.0125 1>}
  n53 = VectorParameter() {group=Bracket, parameterName=Bracket_Color_1, defaultValue=<rgba 0.025 0.0646 0.5 1>}
  n54 = Power(Base=@n17, Exp=ScalarParameter() {defaultValue=2, group=Bracket, parameterName=BracketColorCurve})
  n55 = Subtract(
   A=@n12.R
   B=LinearInterpolate(
    A=ScalarParameter() {defaultValue=0.125, group=Pattern, parameterName=Subtract_Texture_1_Center}
    B=ScalarParameter() {defaultValue=0.333, group=Pattern, parameterName=Subtract_Texture_1_SIdes}
    Alpha=OneMinus(None=@n17)
   )
  )
  n56 = ScalarParameter() {defaultValue=3, group=Pattern, parameterName=Lift_Texture_1}
  n57 = DotProduct(A=Saturate(None=Multiply(A=@n55, B=@n56)), B=@n19)
  LinearInterpolate(
   A=LinearInterpolate(
    A=LinearInterpolate(A=@n1.RGB, B=@n2.RGB, Alpha=@n3)
    B=VectorParameter() {group=Glow, parameterName=Glow_Color_Start, defaultValue=<rgba 1 0.2 0.64 1>}.RGB
    Alpha=Saturate(None=@n27)
   )
   B=LinearInterpolate(
    A=LinearInterpolate(A=@n52.RGB, B=@n53.RGB, Alpha=@n54)
    B=VectorParameter() {group=Bracket, parameterName=Bracket_Spots_Color, defaultValue=<rgba 0.25 0 0.0625 1>}.RGB
    Alpha=Saturate(None=@n57)
   )
   Alpha=@n51
  )
  n58 = ScalarParameter() {defaultValue=1, group=Glow, parameterName=Glow_Start_Alpha}
  Saturate(None=LinearInterpolate(A=ScalarParameter() {group=Glow, parameterName=Glow_Fade_Alpha}, B=Add(A=@n58, B=@n21), Alpha=@n20))
  Saturate(None=Add(A=@n26, B=@n51))
 ]
}
