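# Player health bar indicator widget that shows a parasite gauge, built on AvocadoEaterHealthBarWidget.
# Attach shows it with a bouncing fade-in; detach plays a fade-out, then collapses it when finished.
# Target damage plays a pulse combining transform, opacity and material-parameter tracks for gauge glow, teeth and burst.
# Marker position drives the bracket overlay's horizontal render scale around its right-edge pivot, with a small offset.
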WidgetBlueprint /MotherGameplay/UI/PlayerHealthBarIndicator_Mother
animations [
 Damage_Ani {
  binding_count=5
  length_seconds=1.1849
  object=&self:Damage_Ani
  sequence {
   bindings [
    ParasiteFill_Image {
     id=#1
     possessed_class=Image
     tracks [
      {
       type=MovieScene2DTransformTrack
       sections [{
         type=MovieScene2DTransformSection
         channels [
          Scale.X {type=MovieSceneScriptingFloatChannel, const [0.95,1]}
          Scale.Y {type=MovieSceneScriptingFloatChannel, keys [[0,1.5],[29.32,1]]}
         ]
        }]
      }
      {
       type=MovieSceneColorTrack
       display_name=Color and Opacity
       sections [{
         type=MovieSceneColorSection
         channels [
          Color.R {type=MovieSceneScriptingFloatChannel, const [0.6724,3]}
          Color.G {type=MovieSceneScriptingFloatChannel, keys [[0,0.1746],[10.117,0.4],[29.32,0.1746]]}
          Color.B {type=MovieSceneScriptingFloatChannel, const [0.8632,3]}
          Color.A {type=MovieSceneScriptingFloatChannel, const [1,3]}
         ]
        }]
      }
     ]
    }
    ParasiteGauge_Image {
     id=#2
     possessed_class=Image
     tracks [{
       type=MovieSceneWidgetMaterialTrack
       sections [{
         type=MovieSceneParameterSection
         channels [
          AnimateGlow {type=MovieSceneScriptingFloatChannel, keys [[0,0],[3,1],[30,0]]}
          Glow_Fade_Final_Alpha {type=MovieSceneScriptingFloatChannel, keys [[0,1.125],[5,1.25],[30,1]]}
         ]
        }]
      }]
    }
    ParasiteTeeth_Image {
     id=#3
     possessed_class=Image
     tracks [
      {
       type=MovieSceneFloatTrack
       display_name=Render Opacity
       sections [{type=MovieSceneFloatSection, channels [None {type=MovieSceneScriptingFloatChannel, keys [[0,0],[9,1],[12,1],[18,0]]}]}]
      }
      {
       type=MovieScene2DTransformTrack
       sections [{
         type=MovieScene2DTransformSection
         channels [
          Scale.X {type=MovieSceneScriptingFloatChannel, const [1,1]}
          Scale.Y {type=MovieSceneScriptingFloatChannel, keys [[0,1],[9,1.2],[12,1.2],[18,0.7]]}
         ]
        }]
      }
     ]
    }
    Burst {
     id=#4
     possessed_class=Image
     tracks [
      {
       type=MovieSceneWidgetMaterialTrack
       sections [{
         type=MovieSceneParameterSection
         channels [
          AnimateSubtract {type=MovieSceneScriptingFloatChannel, keys [[0,1],[3,0],[38,1]]}
          Lift_Texture_1 {type=MovieSceneScriptingFloatChannel, keys [[0,3],[10,3],[30,16]]}
          Lift_Texture_2 {type=MovieSceneScriptingFloatChannel, keys [[0,8],[10,8],[30,16]]}
         ]
        }]
      }
      {
       type=MovieScene2DTransformTrack
       sections [{
         type=MovieScene2DTransformSection
         end_frame=71
         start_frame=0
         channels [
          Translation.X {type=MovieSceneScriptingFloatChannel, const [0,1]}
          Translation.Y {type=MovieSceneScriptingFloatChannel, keys [[0,0],[10,32],[30,128]]}
          Scale.X {type=MovieSceneScriptingFloatChannel, keys [[0,2],[60,7]]}
          Scale.Y {type=MovieSceneScriptingFloatChannel, keys [[0,2],[60,7]]}
         ]
        }]
      }
      {
       type=MovieSceneFloatTrack
       display_name=Render Opacity
       sections [{
         type=MovieSceneFloatSection
         end_frame=60
         start_frame=0
         channels [None {type=MovieSceneScriptingFloatChannel, keys [[0,0],[3,1],[45,1],[60,0]]}]
        }]
      }
     ]
    }
    ParasiteSizeBox {
     id=#5
     possessed_class=SizeBox
     tracks [{
       type=MovieScene2DTransformTrack
       sections [{
         type=MovieScene2DTransformSection
         channels [
          Scale.X {type=MovieSceneScriptingFloatChannel, const [1,3]}
          Scale.Y {type=MovieSceneScriptingFloatChannel, keys [[0,1],[9,1.5],[18.001,1]]}
         ]
        }]
      }]
    }
   ]
   display_rate {numerator=60}
  }
 }
 Attach_Ani {
  binding_count=1
  length_seconds=0.35
  object=&self:Attach_Ani
  sequence {bindings [ParasiteSizeBox {
     id=#6
     possessed_class=SizeBox
     tracks [
      {
       type=MovieSceneFloatTrack
       display_name=Render Opacity
       sections [{type=MovieSceneFloatSection, end_frame=7, start_frame=0, channels [None {type=MovieSceneScriptingFloatChannel, keys [[0,0],[3,1],[7,1]]}]}]
      }
      {
       type=MovieScene2DTransformTrack
       sections [{
         type=MovieScene2DTransformSection
         end_frame=7
         start_frame=0
         channels [
          Translation.X {type=MovieSceneScriptingFloatChannel, const [0,4]}
          Translation.Y {type=MovieSceneScriptingFloatChannel, keys [[0,35],[3,-10],[5,5],[7,0]]}
          Angle {type=MovieSceneScriptingFloatChannel, const [0,1]}
          Scale.X {type=MovieSceneScriptingFloatChannel, const [1,2]}
          Scale.Y {type=MovieSceneScriptingFloatChannel, const [1,2]}
          Shear.X {type=MovieSceneScriptingFloatChannel, const [0,1]}
          Shear.Y {type=MovieSceneScriptingFloatChannel, const [0,1]}
         ]
        }]
      }
     ]
    }]}
 }
 Detach_Ani {
  binding_count=1
  length_seconds=0.25
  object=&self:Detach_Ani
  sequence {bindings [ParasiteSizeBox {
     id=#7
     possessed_class=SizeBox
     tracks [
      {
       type=MovieScene2DTransformTrack
       sections [{
         type=MovieScene2DTransformSection
         end_frame=5
         start_frame=0
         channels [
          Translation.X {type=MovieSceneScriptingFloatChannel, keys [[0,0],[5,-25]]}
          Translation.Y {type=MovieSceneScriptingFloatChannel, const [0,1]}
          Scale.X {type=MovieSceneScriptingFloatChannel, keys [[0,1],[5,1.65]]}
          Scale.Y {type=MovieSceneScriptingFloatChannel, keys [[0,1],[5,1.65]]}
         ]
        }]
      }
      {
       type=MovieSceneFloatTrack
       display_name=Render Opacity
       sections [{type=MovieSceneFloatSection, end_frame=5, start_frame=0, channels [None {type=MovieSceneScriptingFloatChannel, keys [[0,1],[5,0]]}]}]
      }
     ]
    }]}
 }
]
graphs [EventGraph {nodes [
   Event OnTargetDamaged() {is=title K2Node_Event}
   > PlayAnimation(InAnimation=Get Damage_Ani() -> (Damage_Ani:Widget Animation Object Reference))
   Event OnParasiteAttached() {is=title K2Node_Event}
   > SetVisibility(InVisibility=SelfHitTestInvisible)
   > StopAnimation(InAnimation=Get Damage_Ani())
   > StopAnimation(InAnimation=Get Detach_Ani() -> (Detach_Ani:Widget Animation Object Reference))
   > PlayAnimation(InAnimation=Get Attach_Ani() -> (Attach_Ani:Widget Animation Object Reference))
   Event OnParasiteDetached() {is=title K2Node_Event}
   > StopAnimation(InAnimation=Get Damage_Ani())
   > StopAnimation(InAnimation=Get Attach_Ani())
   > PlayAnimation(InAnimation=Get Detach_Ani(), bRestoreState=true)
   n1 = Event OnSetMarkerPosition!() -> (OutputDelegate, then, MarkerPosition:float) {is=title K2Node_Event}
   > SetRenderScale!(
    execute
    self:Widget Object Reference=Get ParasiteBracketOverlay() -> (ParasiteBracketOverlay:Overlay Object Reference)
    Scale_X:double=float - float(B=float - float(A=@n1.MarkerPosition, B=0.015)) {is=title K2Node_CallFunction}
    Scale_Y:double=1
   ) -> (then)
   "Animation Finished (Detach_Ani)"!() -> (OutputDelegate:Delegate, then:Exec) {is=title K2Node_WidgetAnimationEvent}
   > SetVisibility(InVisibility=Collapsed)
   float + float(A=@n1.MarkerPosition, B=0.001) {is=title K2Node_CommutativeAssociativeBinaryOperator}
  ]}]
variables []
widgets {
 info {inheritedWidgetCount=1, parentClass=/Script/MotherGameplayUI.AvocadoEaterHealthBarWidget, rootWidgetClass=&SizeBox, widgetCount=7}
 widgets [
  ParasiteSizeBox {widgetClassPath=&SizeBox, widget=ParasiteSizeBox, properties {heightOverride=64, widthOverride=150}}
  ParasiteBracketOverlay {
   widgetClassPath=&Overlay
   bIsVariable=true
   parent=ParasiteSizeBox
   widget=ParasiteBracketOverlay
   properties {renderTransformPivot=<1 0.5>}
   slot {class=SizeBoxSlot}
  }
  Burst {
   widgetClassPath=&Image
   parent=ParasiteBracketOverlay
   widget=Burst
   properties {
    renderOpacity=0
    visibility=SelfHitTestInvisible
    brush {
     bIsDynamicallyLoaded=false
     drawAs=Image
     imageType=NoImage
     mirroring=NoMirror
     resourceName=None
     resourceObject=&/M_UI_Mother_HealthBar_BurstFX_Inst
     tiling=NoTile
     imageSize=<32 32>
     margin=<ltrb 0 0 0 0>
     outlineSettings {
      bUseBrushTransparency=false
      roundingType=HalfHeightRadius
      width=0
      color {colorUseRule=UseColor_Specified, specifiedColor=<rgba 0 0 0 0>}
      cornerRadii=<0 0 0 0>
     }
     tintColor {colorUseRule=UseColor_Specified, specifiedColor=<rgba 1 1 1 1>}
     uVRegion {bIsValid=false, max=<0 0>, min=<0 0>}
    }
    renderTransform {angle=0, scale=<2 2>, shear=<0 0>, translation=<0 0>}
   }
   slot {class=OverlaySlot, properties {horizontalAlignment=HAlign_Center, verticalAlignment=VAlign_Center}}
  }
  ParasiteTeeth_Image {
   widgetClassPath=&Image
   parent=ParasiteBracketOverlay
   widget=ParasiteTeeth_Image
   properties {
    renderOpacity=0
    visibility=HitTestInvisible
    brush {
     bIsDynamicallyLoaded=false
     drawAs=Image
     imageType=NoImage
     mirroring=NoMirror
     resourceName=None
     resourceObject=/MotherGameplay/UI/Textures/T_UI_ParasiteGauge_Teeth
     tiling=NoTile
     imageSize=<256 64>
     margin=<ltrb 0 0 0 0>
     outlineSettings {
      bUseBrushTransparency=false
      roundingType=HalfHeightRadius
      width=0
      color {colorUseRule=UseColor_Specified, specifiedColor=<rgba 0 0 0 0>}
      cornerRadii=<0 0 0 0>
     }
     tintColor {colorUseRule=UseColor_Specified, specifiedColor=<rgba 1 1 1 1>}
     uVRegion {bIsValid=false, max=<0 0>, min=<0 0>}
    }
    renderTransform {angle=0, scale=<1 1.1>, shear=<0 0>, translation=<0 2>}
    renderTransformPivot=<0.5 1>
   }
   slot {class=OverlaySlot, properties {horizontalAlignment=HAlign_Fill, verticalAlignment=VAlign_Bottom}}
  }
  ParasiteFill_Image {
   widgetClassPath=&Image
   parent=ParasiteBracketOverlay
   widget=ParasiteFill_Image
   properties {
    renderOpacity=0.75
    visibility=Collapsed
    brush {
     bIsDynamicallyLoaded=false
     drawAs=Image
     imageType=NoImage
     mirroring=NoMirror
     resourceName=None
     resourceObject=/FortniteCoreUI/UI/Login/SubGameSelect/Materials/M_UI_TileRadialGradient_Center
     tiling=NoTile
     imageSize=<128 24>
     margin=<ltrb 0 0 0 0>
     outlineSettings {
      bUseBrushTransparency=false
      roundingType=HalfHeightRadius
      width=0
      color {colorUseRule=UseColor_Specified, specifiedColor=<rgba 0 0 0 0>}
      cornerRadii=<0 0 0 0>
     }
     tintColor {colorUseRule=UseColor_Specified, specifiedColor=<rgba 1 1 1 1>}
     uVRegion {bIsValid=false, max=<0 0>, min=<0 0>}
    }
    colorAndOpacity=<rgba 0.6724 0.1746 0.8632 1>
    renderTransform {angle=0, scale=<0.95 1>, shear=<0 0>, translation=<0 -4>}
    renderTransformPivot=<0.5 1>
   }
   slot {class=OverlaySlot, properties {horizontalAlignment=HAlign_Fill, verticalAlignment=VAlign_Center}}
  }
  ParasiteGauge_Image {
   widgetClassPath=&Image
   parent=ParasiteBracketOverlay
   widget=ParasiteGauge_Image
   properties {
    visibility=SelfHitTestInvisible
    brush {
     bIsDynamicallyLoaded=false
     drawAs=Image
     imageType=NoImage
     mirroring=NoMirror
     resourceName=None
     resourceObject=&/M_UI_Mother_HealthBar_Indicator
     tiling=NoTile
     imageSize=<256 64>
     margin=<ltrb 0 0 0 0>
     outlineSettings {
      bUseBrushTransparency=false
      roundingType=HalfHeightRadius
      width=0
      color {colorUseRule=UseColor_Specified, specifiedColor=<rgba 0 0 0 0>}
      cornerRadii=<0 0 0 0>
     }
     tintColor {colorUseRule=UseColor_Specified, specifiedColor=<rgba 1 1 1 1>}
     uVRegion {bIsValid=false, max=<0 0>, min=<0 0>}
    }
    renderTransform {angle=0, scale=<1 1.5>, shear=<0 0>, translation=<0 0>}
   }
   slot {class=OverlaySlot, properties {horizontalAlignment=HAlign_Fill, verticalAlignment=VAlign_Fill}}
  }
  TutorialHighlight {widgetClassPath=&TutorialHighlight}
 ]
}
